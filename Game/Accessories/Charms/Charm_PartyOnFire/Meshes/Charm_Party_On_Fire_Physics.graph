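# Physics asset for the Party On Fire charm accessory mesh, defining its collision bodies and joints for simulated dangling motion.
# A box body on the root anchors a four-link chain: three small boxes followed by a larger rotated capsule on the last bone.
# Constraints join each link to its parent in series, allowing moderate swing but very little twist.
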PhysicsAsset /Game/Accessories/Charms/Charm_PartyOnFire/Meshes/Charm_Party_On_Fire_Physics
bodies {
 chain_01 [{shapeType=Box, extentX=1.01, extentY=1.01, extentZ=1.01, center=<1 0 0>}]
 chain_02 [{shapeType=Box, extentX=1.01, extentY=1.01, extentZ=1.01, center=<1 0 0>}]
 chain_03 [{shapeType=Box, extentX=1.01, extentY=1.01, extentZ=1.01, center=<1 0 0>}]
 chain_04 [{shapeType=Capsule, length=2.1333, radius=2.8419, center=<4.4837 0.005 -0.5703>, rotation=<rot 90 0 0>}]
 root [{shapeType=Box, extentX=1.9926, extentY=2.7676, extentZ=2.8437, center=<-0.2499 0 0>}]
}
constraints [
 {bone1Name=chain_01, bone2Name=root, swing1LimitDegrees=45, swing2LimitDegrees=45, twistLimitDegrees=5}
 {bone1Name=chain_02, bone2Name=chain_01, swing1LimitDegrees=45, swing2LimitDegrees=45, twistLimitDegrees=5}
 {bone1Name=chain_03, bone2Name=chain_02, swing1LimitDegrees=45, swing2LimitDegrees=45, twistLimitDegrees=5}
 {bone1Name=chain_04, bone2Name=chain_03, swing1LimitDegrees=45, swing2LimitDegrees=45, twistLimitDegrees=5}
]
